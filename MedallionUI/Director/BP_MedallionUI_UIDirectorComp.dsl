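; UI director component for Fortnite medallions that adds and removes their HUD, inventory and message scenes.
; Playlist tags and console variables gate inventory scenes; teardown omits the Mash scene.
; The HUD scene follows the view target's augment list, rebinding on view-target changes.
; Tracking-component events show boss or minor messages until a routed event removes them, though UnbindMedallionTrackerEvents calls Bind instead of Unbind.
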(event EventOnReadyToShowUI
  (CallFunction|BindViewTargetEvents)
  (CallFunction|BindMedallionTrackerEvents)
  (Default|AddInventoryScenes))

(event EventEndPlay
  (params (EndPlayReason :type EEndPlayReason%20Enum :default "Destroyed"))
  (CallFunction|UnbindViewTargetEvents)
  (Default|UnbindMedallionTrackerEvents)
  (Default|RemoveInventoryScenes))

(event Custom|RegisterMedallionMessageRemoveEvent
  (bind (_asynctaskproxy _payload) (GameplayEventRouter|StartListeningToEvent (GameplayEventRouter|GetGlobalEventRouter) "/MedallionUI/Message/BPEvent_RemoveMedallionMessaging.BPEvent_RemoveMedallionMessaging" (CallFunction|GetOwningLocalPlayerController))
    (:then
      (Variables|Default|SetRemoveMedallioMsgHandle _asynctaskproxy))
    (:OnEventReceived
      (BaseMedallions|HideMedallionMessageScene (select (Utilities|Struct|BreakBPEventRemoveMedallionMessaging _payload) (Variables|Default|GetMinorMedallionMessageScene) (Variables|BossMedallions|GetBossMedallionMessageScene)))
      (GameplayEventRouter|StopListeningtoEvent (Variables|Default|GetRemoveMedallioMsgHandle)))))

(fn AddInventoryScenes
  (Utilities|FlowControl|Sequence
    (:then_0
      (if (Default|IsPlaylistActive (Variables|Default|GetPlaylists_MedallionInventoryScene))
        (CallFunction|AddScene (Variables|Default|GetInventoryScene))))
    (:then_1
      (bind _returnvalue (Development|GetConsoleVariableBoolValue "Athena.Inventory.Agument.MinorMedallions"))
      (if (and _returnvalue (Default|IsPlaylistActive (Variables|Default|GetPlaylists_MinorMedallionInventoryScene)))
        (CallFunction|AddScene (Variables|Default|GetInventoryMinorScene))))
    (:then_2
      (bind _returnvalue_2 (Development|GetConsoleVariableBoolValue "Athena.Inventory.Agument.ForceAbilities"))
      (if (and _returnvalue_2 (Default|IsPlaylistActive (Variables|Default|GetPlaylists_PlotTwistInventoryScene)))
        (CallFunction|AddScene (Variables|Default|GetInventoryPlotTwistScene))))
    (:then_3
      (if (Default|IsPlaylistActive (Variables|Default|GetPlaylists_Mash))
        (CallFunction|AddScene (Variables|Default|GetInventoryScene_Mash))))))

(fn BindMedallionTrackerEvents
  (if (not (Default|IsPlaylistActive (Variables|Default|GetPlaylists_ExcludeMedallionTracking)))
    (bind _output_get (LocalVariables|SetMedallionsTrackingComponent (Actor|GetComponentbyClass (CallFunction|GetOwningLocalPlayerController) "/Medallions/BP_Medallions_TrackingComponent.BP_Medallions_TrackingComponent_C")))
    (Default|BindEventtoOnMedallionCountChanged _output_get (EventDispatchers|CreateEvent :!event HandleMedallionCountChanged))
    (Default|BindEventtoOnMinorMedallionAdded (LocalVariables|GetMedallionsTrackingComponent) (EventDispatchers|CreateEvent :!event HandleMinorMedallionAdded))))

(fn BindViewTargetEvents
  (bind _asathena_player_controller (Utilities|Casting|CastToAthena_PlayerController (CallFunction|GetOwningLocalPlayerController)))
  (EventDispatchers|BindEventtoOnViewTargetChanged _asathena_player_controller (EventDispatchers|CreateEvent :!event HandleViewTargetChanged))
  (CallFunction|HandleViewTargetChanged 0 0 (Pawn|GetViewTarget (CallFunction|GetOwningLocalPlayerController))))

(fn HandleAugmentListChanged
  (params (AugmentComponent :type Fort%20Player%20State%20Component%20Player%20Augment%20System%20Object%20Reference))
  (bind _medallionhudscene (Variables|Default|GetMedallionHUDScene))
  (bind _returnvalue (Class|FortPlayerStateComponentPlayerAugmentSystem|GetActiveAugments (Components|GetOwner AugmentComponent)))
  (if (> (Utilities|Array|Length _returnvalue) 0)
    (if (not (DynamicUIDirector|IsSceneActive (CallFunction|GetOwningFortDynamicUIDirector) (Variables|Default|GetMedallionHUDScene)))
      (CallFunction|AddScene _medallionhudscene))
    (else
      (if (DynamicUIDirector|IsSceneActive (CallFunction|GetOwningFortDynamicUIDirector) (Variables|Default|GetMedallionHUDScene))
        (CallFunction|RemoveScene _medallionhudscene)))))

(fn HandleMedallionCountChanged
  (params (DidCountIncrease :type Boolean) (NewCount :type Integer) (NewestMedallionAdded :type Fort%20Player%20Augment%20Item%20Definition%20Object%20Reference))
  (BaseMedallions|ShowMedallionMessageScene (Variables|BossMedallions|GetBossMedallionMessageScene)))

(fn HandleMinorMedallionAdded
  (params (NewlyAddedMinorMedallion :type Fort%20Player%20Augment%20Item%20Definition%20Object%20Reference))
  (BaseMedallions|ShowMedallionMessageScene (Variables|Default|GetMinorMedallionMessageScene)))

(fn HandleViewTargetChanged
  (params (PlayerController :type Fort%20Player%20Controller%20Object%20Reference) (OldViewTarget :type Actor%20Object%20Reference) (NewViewTarget :type Actor%20Object%20Reference))
  (bind _newviewtarget (Variables|GetNewViewTarget))
  (bind _returnvalue (Class|FortPlayerStateComponentPlayerAugmentSystem|GetAugmentComponent _newviewtarget))
  (Utilities|FlowControl|Sequence
    (:then_0
      (Default|UnbindAugmentMedallionEvents (Variables|GetOldViewTarget)))
    (:then_1
      (Utilities|IsValid (Variables|GetNewViewTarget)
        (:Is%20Valid
          (Utilities|IsValid _returnvalue
            (:Is%20Valid
              (EventDispatchers|BindEventtoOnAugmentListUpdated _returnvalue (EventDispatchers|CreateEvent :!event HandleAugmentListChanged)))))))))

(fn HideMedallionMessageScene
  (params (Scene :type Dynamic%20UIScene%20Object%20Reference))
  (if (DynamicUIDirector|IsSceneActive (CallFunction|GetOwningFortDynamicUIDirector) (Variables|GetScene))
    (CallFunction|RemoveScene (Variables|GetScene))))

(fn IsPlaylistActive
  (params (PlaylistTags :type Gameplay%20Tag%20Container%20Structure%20%28by%20ref%29))
  (bind _asfort_game_state_athena (Utilities|Casting|CastToFortGameStateAthena (Game|GetGameState))
    (:then
      (return (GameplayTags|HasAnyTags (Class|FortGameStateAthena|GetAthenaPlaylistContextTags _asfort_game_state_athena) (Variables|GetPlaylistTags))))
    (:CastFailed
      (return false))))

(fn RemoveInventoryScenes
  (Utilities|FlowControl|Sequence
    (:then_0
      (if (Default|IsPlaylistActive (Variables|Default|GetPlaylists_MedallionInventoryScene))
        (CallFunction|RemoveScene (Variables|Default|GetInventoryScene))))
    (:then_1
      (bind _returnvalue (Development|GetConsoleVariableBoolValue "Athena.Inventory.Agument.MinorMedallions"))
      (if (and _returnvalue (Default|IsPlaylistActive (Variables|Default|GetPlaylists_MinorMedallionInventoryScene)))
        (CallFunction|RemoveScene (Variables|Default|GetInventoryMinorScene))))
    (:then_2
      (bind _returnvalue_2 (Development|GetConsoleVariableBoolValue "Athena.Inventory.Agument.ForceAbilities"))
      (if (and _returnvalue_2 (Default|IsPlaylistActive (Variables|Default|GetPlaylists_PlotTwistInventoryScene)))
        (CallFunction|RemoveScene (Variables|Default|GetInventoryPlotTwistScene))))))

(fn ShowMedallionMessageScene
  (params (Scene :type Dynamic%20UIScene%20Object%20Reference))
  (if (not (DynamicUIDirector|IsSceneActive (CallFunction|GetOwningFortDynamicUIDirector) (Variables|GetScene)))
    (CallFunction|AddScene (Variables|GetScene))
    (CallFunction|RegisterMedallionMessageRemoveEvent)))

(fn UnbindAugmentMedallionEvents
  (params (ViewTarget :type Actor%20Object%20Reference))
  (bind _viewtarget (Variables|GetViewTarget))
  (bind _returnvalue (Class|FortPlayerStateComponentPlayerAugmentSystem|GetAugmentComponent _viewtarget))
  (Utilities|IsValid ViewTarget
    (:Is%20Valid
      (Utilities|IsValid _returnvalue
        (:Is%20Valid
          (EventDispatchers|UnbindEventfromOnAugmentListUpdated _returnvalue (EventDispatchers|CreateEvent :!event HandleAugmentListChanged)))))))

(fn UnbindMedallionTrackerEvents
  (if (not (Default|IsPlaylistActive (Variables|Default|GetPlaylists_ExcludeMedallionTracking)))
    (bind _output_get (LocalVariables|SetOwningPlayerController (CallFunction|GetOwningLocalPlayerController)))
    (Utilities|IsValid _output_get
      (:Is%20Valid
        (bind _output_get_2 (LocalVariables|SetMedallionsTrackerComp (Actor|GetComponentbyClass _output_get "/Medallions/BP_Medallions_TrackingComponent.BP_Medallions_TrackingComponent_C")))
        (Utilities|IsValid _output_get_2
          (:Is%20Valid
            (Default|BindEventtoOnMedallionCountChanged (LocalVariables|GetMedallionsTrackerComp) (EventDispatchers|CreateEvent :!event HandleMedallionCountChanged))
            (Default|BindEventtoOnMinorMedallionAdded (LocalVariables|GetMedallionsTrackerComp) (EventDispatchers|CreateEvent :!event HandleMinorMedallionAdded))))))))

(fn UnbindViewTargetEvents
  (bind _asathena_player_controller (Utilities|Casting|CastToAthena_PlayerController (CallFunction|GetOwningLocalPlayerController)))
  (bind _output_get (LocalVariables|SetLocalPlayerController _asathena_player_controller))
  (EventDispatchers|UnbindEventfromOnViewTargetChanged _output_get (EventDispatchers|CreateEvent :!event HandleViewTargetChanged))
  (Default|UnbindAugmentMedallionEvents (Pawn|GetViewTarget (LocalVariables|GetLocalPlayerController))))
